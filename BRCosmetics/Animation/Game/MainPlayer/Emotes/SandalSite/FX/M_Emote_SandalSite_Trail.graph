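# Unlit, two-sided translucent Niagara trail material for the SandalSite emote, usable on ribbons, sprites and meshes.
# Particle color drives emissive through MF_FX_EmissiveAdjustments.
# Opacity blends panning noise with a dynamic-parameter-shifted U gradient, shaped by UV edge masks and an inverted sphere mask, then scaled by particle alpha and view-angle fade.
# A shading path switch applies depth fade on default and deferred paths, skipping it on forward and mobile.
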Material /BRCosmetics/Animation/Game/MainPlayer/Emotes/SandalSite/FX/M_Emote_SandalSite_Trail
graph {
 properties {
  bAutomaticallySetUsageInEditor=false
  bUsedWithNiagaraMeshParticles=true
  bUsedWithNiagaraRibbons=true
  bUsedWithNiagaraSprites=true
  blendMode=BLEND_Translucent
  shadingModel=MSM_Unlit
  twoSided=true
  usedShadingModels=Unlit
 }
 nodes [
  n1 = ParticleColor()
  MaterialFunctionCall(Emissive=@n1.RGB) {materialFunction=&MF_FX_EmissiveAdjustments}
  n2 = TextureCoordinate()
  n3 = AppendVector(
   A=ScalarParameter() {defaultValue=0.25, parameterName=BaseTexture_UTiling}
   B=ScalarParameter() {defaultValue=2, parameterName=BaseTexture_VTiling}
  )
  n4 = TextureSampleParameter2D(UVs=Panner(Coordinate=Multiply(A=@n2, B=@n3)) {speedX=-0.5}) {parameterName=BaseTexture, texture=&T_NoisePack_01}
  n5 = DynamicParameter() {defaultValue=<rgba 0 1 1 1>, paramNames [MaskU_Add,Param2,Param3,Param4]}
  n6 = ScalarParameter() {parameterName=MaskV_Add}
  n7 = ComponentMask(None=Add(A=TextureCoordinate(), B=AppendVector(A=@n5.MaskU_Add, B=@n6))) {r=true}
  n8 = Add(A=Multiply(A=Add(A=@n4.R) {constB=-0.6}) {constB=0.5}, B=Multiply(A=OneMinus(None=@n7)) {constB=0.5})
  n9 = ComponentMask(None=TextureCoordinate()) {g=true}
  n10 = OneMinus(None=@n9)
  n11 = Multiply(A=Multiply(A=@n9) {constB=10}, B=Power(Base=Multiply(A=@n10)) {constExponent=3})
  n12 = ComponentMask(None=TextureCoordinate()) {r=true}
  n13 = OneMinus(None=@n12)
  n14 = Multiply(
   A=Multiply(A=@n12, B=ScalarParameter() {defaultValue=10, parameterName=MaskIntensity}) {constB=10}
   B=Power(Base=Multiply(A=@n13)) {constExponent=3}
  )
  n15 = SphereMask(
   A=TextureCoordinate()
   B=AppendVector(
    A=ScalarParameter() {defaultValue=0.8, parameterName=SphereMaskOffset_X}
    B=ScalarParameter() {defaultValue=0.8, parameterName=SphereMaskOffset_Y}
   )
  ) {attenuationRadius=0.8, hardnessPercent=0.3}
  n16 = Multiply(A=Multiply(A=Saturate(None=@n11), B=Saturate(None=@n14)), B=Power(Base=OneMinus(None=@n15)) {constExponent=3})
  n17 = Multiply(
   A=Multiply(A=@n1.A, B=Multiply(A=@n8, B=@n16))
   B=MaterialFunctionCall("FadeSensitivity (optional)"=ScalarParameter() {defaultValue=0.8, parameterName=MeshFade}) {materialFunction=&FadeBasedOnViewAngle}.Result
  )
  n18 = DepthFade(Opacity=@n17, FadeDistance=ScalarParameter() {defaultValue=5, parameterName=DepthFade})
  ShadingPathSwitch(Default=@n18, Deferred=@n18, Forward=@n17, Mobile=@n17)
 ]
}
