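# Niagara system for the big ground impact effect of the SandalSite emote, a burst of dirt, smoke and stones with a flash.
# Separate burst emitters layer dirt, soil, stones, curl-noise smoke, radial and directional streaks, an expanding mesh ground ring, and a glow with a particle light.
# Curves drive scale, fade and dynamic material parameters, while brightness, size and opacity adapt to time of day, platform and front-end preview in several emitters.
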NiagaraSystem /BRCosmetics/Animation/Game/MainPlayer/Emotes/SandalSite/FX/NS_Emote_SandalSite_ImpactBig
dependencies {
 usedDataInterfaces [
  &NiagaraDataInterfaceMeshRendererInfo
  &NiagaraDataInterfaceCurve
  &NiagaraDataInterfacePlatformSet
  &NiagaraDataInterfaceColorCurve
  &NiagaraDataInterfaceArrayFloat
  &NiagaraDataInterfaceVector2DCurve
  &NiagaraDataInterfaceArrayFloat2
  &NiagaraDataInterfaceArrayFloat3
 ]
 usedDynamicInputs [
  &RandomRangeInt
  &SelectFloatFromArray
  &SpawnGroup
 ]
 usedModules [&SystemState]
}
emitters {
 Dirt {
  emitterUpdateScript [
   &EmitterState(Inactive Response=::NewEnumerator1, Life Cycle Mode=::NewEnumerator1, Loop Behavior=::NewEnumerator1, Loop Duration=1, UseLoopDelay=false)
   &SpawnBurst_Instantaneous(Spawn Count=15, Use Spawn Probability=false, UseLoopCountLimit=false)
  ]
  particleSpawnScript [
   &InitializeParticle(
    Color Maximum=<rgba 0.4427 0.2595 0.0853 1>
    Color Minimum=<rgba 0.276 0.1052 0.0331 1>
    Color Mode=::NewEnumerator2
    Lifetime Max=1.5
    Lifetime Min=0.5
    Lifetime Mode=::NewEnumerator1
    Sprite Size Max=<2 6>
    Sprite Size Min=<0.5 2>
    Sprite Size Mode=::NewEnumerator2
    UsePositionOffset=true
    Write Material Random=true
   )
   &ShapeLocation(
    Large Radius=50
    MaskSpawnGroup=false
    Non Uniform Scale=<1 1 0.5>
    Offset Mode=::NewEnumerator0
    Override Seed=false
    Shape Primitive=::NewEnumerator3
    UseOwnerScale=true
   )
   &AddVelocity(
    Module.Use Falloff Distance=false
    Origin Offset=<0 0 -50>
    Override Seed=false
    Velocity Mode=::NewEnumerator1
    Velocity Speed=&RandomRangeFloat(Minimum=150, Maximum=300)
   )
   &AddVelocity(Velocity=<0 0 20>) {module=AddVelocity001}
  ]
  particleUpdateScript [
   &ParticleState()
   &ScaleSpriteSize(Scale Factor=<0.5 1>, Scale Sprite Size Mode=::NewEnumerator2) {enabled=false, module=ScaleSpriteSize001}
   &ScaleSpriteSize(Uniform Curve Sprite Scale=DataInterface() {propertyValues {Curve {keys [
        {arriveTangent=8.2127, interpMode=RCIM_Cubic, leaveTangent=8.2127, tangentMode=RCTM_User, time=0, value=0.5}
        {
         arriveTangent=4.2448
         arriveTangentWeight=0.232
         interpMode=RCIM_Cubic
         leaveTangent=4.2448
         leaveTangentWeight=2.2538
         tangentMode=RCTM_User
         time=0.07
         value=1
        }
        {arriveTangent=-2.5194, interpMode=RCIM_Cubic, leaveTangent=-2.5194, tangentMode=RCTM_User, time=0.6, value=0.8}
        {arriveTangent=-0.2055, interpMode=RCIM_Cubic, leaveTangent=-0.2055, tangentMode=RCTM_User, time=1, value=0}
       ]}}})
   &DynamicMaterialParameters(
    Index 0 Param 1=&FloatFromCurve(
     FloatCurve=DataInterface() {propertyValues {Curve {keys [
         {arriveTangent=-5.8452, interpMode=RCIM_Cubic, time=0, value=1}
         {arriveTangentWeight=0.0434, interpMode=RCIM_Cubic, leaveTangentWeight=0.217, tangentMode=RCTM_User, time=0.2, value=0}
         {arriveTangent=3.4164, interpMode=RCIM_Cubic, leaveTangent=3.4164, tangentMode=RCTM_User, time=0.8, value=0.7}
         <at 1 0 RCIM_Cubic>
        ]}}}
     Scale Curve=3
    )
    Index 0 Param 3=1
    Index 0 Param 4=&FloatFromCurve(
     FloatCurve=DataInterface() {propertyValues {Curve {keys [
         {interpMode=RCIM_Cubic, leaveTangent=-1.5195, tangentMode=RCTM_Break, time=0, value=1}
         {arriveTangent=0.0067, interpMode=RCIM_Cubic, tangentMode=RCTM_Break, time=1, value=0}
        ]}}}
     Scale Curve=&RandomRangeFloat(Minimum=10, Maximum=30)
    )
    Module.Param0WriteEnabled=true
    Module.Param1WriteEnabled=true
    Module.Param2WriteEnabled=true
    Module.Param3WriteEnabled=true
   )
   &Drag(Drag=&RandomRangeFloat(Minimum=1, Maximum=2), Use Linear Drag=true, Use Rotational Drag=true)
   &GravityForce(Gravity=&RandomRangeVector(Minimum=<0 0 -300>, Maximum=<0 0 -600>))
   &SolveForcesAndVelocity(Clamp Velocity=false, Limit Acceleration=false)
   &ScaleColor(
    Scale Alpha=&Lerp_Float(
     A=0
     B=1
     Alpha=&FloatFromCurve(
      FloatCurve=DataInterface() {propertyValues {Curve {keys [
          {interpMode=RCIM_Cubic, tangentMode=RCTM_User, time=0, value=1}
          {interpMode=RCIM_Cubic, tangentMode=RCTM_User, time=0.5, value=1}
          {tangentMode=RCTM_User, time=1, value=0}
         ]}}}
      Scale Curve=&Multiply_Float(A=1, B=&FNTimeOfDayScalar(Night Scale=0.6, Morning Scale=0.9, Evening Scale=0.9))
     )
    )
    ScaleA=true
    ScaleRGB=true
   )
  ]
  rendererClasses [&NiagaraSpriteRendererProperties]
  renderers [{rendererClass=&NiagaraSpriteRendererProperties}]
 }
 Glow {
  emitterUpdateScript [
   &EmitterState(
    Inactive Response=::NewEnumerator1
    Life Cycle Mode=::NewEnumerator1
    Loop Behavior=::NewEnumerator1
    Loop Delay=1.6
    Loop Duration=0.5
    UseLoopDelay=false
   )
   &SpawnBurst_Instantaneous(Use Spawn Probability=false, UseLoopCountLimit=false)
  ]
  particleSpawnScript [
   &InitializeParticle(
    Color=<rgba 1 0.5485 0.0896 1>
    Lifetime=0.35
    Sprite Rotation Mode=::NewEnumerator2
    Sprite Size Mode=::NewEnumerator3
    Uniform Sprite Size=200
    UsePositionOffset=true
    Write Material Random=true
   )
   &CameraOffset(Camera Offset Amount=&FN_PickFloatDependOnPlatform(TargetPlatformActive=30, TargetPlatformNoActive=80))
   /SetVariables() {isSetParameters=true}
   /SetVariables(Particles.Position002:NiagaraPosition=<0 0 5>) {isSetParameters=true}
  ]
  particleUpdateScript [
   &ParticleState()
   &ScaleSpriteSize(
    Uniform Curve Scale=&FN_FrontEndFloatModifier(
     Gameplay=&FN_PickFloatDependOnPlatform(TargetPlatformActive=1, TargetPlatformNoActive=0.7)
     Front End=0.8
    )
    Uniform Curve Sprite Scale=DataInterface() {propertyValues {Curve {keys [
        {arriveTangent=5.1246, interpMode=RCIM_Cubic, leaveTangent=5.1246, tangentMode=RCTM_User, time=0, value=0}
        {arriveTangent=1.7949, interpMode=RCIM_Cubic, leaveTangent=1.7949, tangentMode=RCTM_User, time=0.3, value=1.5}
        <at 1 2 RCIM_Cubic>
       ]}}}
   )
   /SetVariables(Particles.Size002:Vector2f=&Multiply_Vector2D(A=$Particles.SpriteSize, B=<1.2 0.7>)) {isSetParameters=true}
   &ScaleColor(
    Scale Alpha=&FloatFromCurve(
     FloatCurve=DataInterface() {propertyValues {Curve {keys [
         {arriveTangent=-2.7027, interpMode=RCIM_Cubic, leaveTangent=-2.7027, tangentMode=RCTM_User, time=0, value=1}
         {arriveTangent=-1, interpMode=RCIM_Cubic, leaveTangent=-1, time=0.312, value=0.3309}
         <at 1 0>
        ]}}}
     Scale Curve=&Multiply_Float(A=1, B=&FNTimeOfDayScalar(Night Scale=0.8, Morning Scale=1, Evening Scale=1))
    )
    ScaleA=true
    ScaleRGB=true
   )
   /SetVariables(Particles.Color002=&MakeLinearColorFromVectorAndFloat(
     "Vector (RGB)"=&MakeVectorFromLinearColorRGB(Color=<rgba 1 0.5 0.0896 1>)
     "Float (Alpha)"=&FloatFromCurve(
      FloatCurve=DataInterface() {propertyValues {Curve {keys [
          {arriveTangent=-2.7027, interpMode=RCIM_Cubic, leaveTangent=-2.7027, tangentMode=RCTM_User, time=0, value=1}
          {arriveTangent=-1.1238, interpMode=RCIM_Cubic, leaveTangent=-1.1238, tangentMode=RCTM_User, time=0.3, value=0.3}
          <at 1 0>
         ]}}}
      Scale Curve=&Multiply_Float(A=1, B=&FNTimeOfDayScalar(Night Scale=0.8, Morning Scale=1, Evening Scale=1))
     )
    )) {isSetParameters=true}
   &FN_Light_Attributes(
    Color=&ColorFromCurve(ColorCurve=DataInterface() {propertyValues {
       AlphaCurve {keys []}
       BlueCurve {keys [<at 0.2351 0.3 RCIM_Cubic>,<at 0.5132 0.1719>]}
       GreenCurve {keys [<at 0.2351 1.01 RCIM_Cubic>,<at 0.5132 0.5705>]}
       RedCurve {keys [<at 0.2351 2 RCIM_Cubic>,<at 0.5132 1>]}
      }})
    Create Color Attribute=true
    Create Exponent Attribute=true
    Create Position Attribute=true
    Create Radius Attribute=true
    Enable Particle Light Enable=true
    Exponent=3
    Position=<0 0 50>
    Radius=&Multiply_Float(
     A=60
     B=&FloatFromCurve(FloatCurve=DataInterface() {propertyValues {Curve {keys [<at 0 1.8>,<at 0.2 2.05>,<at 0.5 2.2>,<at 1 1.8>]}}})
    )
    Scale Color=&FloatFromCurve(
     FloatCurve=DataInterface() {propertyValues {Curve {keys [
         {arriveTangent=5.626, interpMode=RCIM_Cubic, leaveTangent=5.626, tangentMode=RCTM_User, time=0, value=0.6}
         {arriveTangent=0.0895, interpMode=RCIM_Cubic, leaveTangent=0.0895, tangentMode=RCTM_User, time=0.1, value=1}
         {arriveTangent=-0.8586, interpMode=RCIM_Cubic, leaveTangent=-0.8586, tangentMode=RCTM_User, time=0.815, value=0.72}
         {arriveTangent=-4.4162, interpMode=RCIM_Cubic, leaveTangent=-4.4162, tangentMode=RCTM_User, time=1, value=0}
        ]}}}
     Scale Curve=&Multiply_Float(
      A=&FN_FrontEndFloatModifier(Gameplay=20, Front End=5)
      B=&FNTimeOfDayScalar(Night Scale=0.5, Morning Scale=1, Evening Scale=0.8)
     )
    )
   )
  ]
  rendererClasses [&NiagaraSpriteRendererProperties, &NiagaraLightRendererProperties]
  renderers [
   {rendererClass=&NiagaraSpriteRendererProperties}
   {rendererClass=&NiagaraLightRendererProperties, rendererIndex=1}
   {rendererClass=&NiagaraSpriteRendererProperties, rendererIndex=2}
  ]
 }
 Ground_Ring {
  emitterUpdateScript [
   &EmitterState(Inactive Response=::NewEnumerator1, Life Cycle Mode=::NewEnumerator1, Loop Behavior=::NewEnumerator1, Loop Duration=1, UseLoopDelay=false)
   &SpawnBurst_Instantaneous(Use Spawn Probability=false, UseLoopCountLimit=false)
  ]
  particleSpawnScript [
   &InitializeParticle(Lifetime=0.4, Mesh Scale=<0.5 1.5 1.5>, Mesh Scale Mode=::NewEnumerator4, UsePositionOffset=true, Write Material Random=true)
   &InitialMeshOrientation(Mesh Orientation Mode=::NewEnumerator2, Rotation=<0 0.25 0>, Use Rotation Vector=true)
   /SetVariables() {enabled=false, isSetParameters=true}
   /SetVariables(Particles.Position002:NiagaraPosition=<0 0 5>) {enabled=false, isSetParameters=true}
   &ShapeLocation(
    Box Size=<0 0 1>
    MaskSpawnGroup=false
    Offset=<0 0 10>
    Offset Mode=::NewEnumerator0
    Override Seed=false
    Shape Primitive=::NewEnumerator2
    Surface Only=false
    UseOwnerScale=true
   )
  ]
  particleUpdateScript [
   &ParticleState()
   &ScaleMeshSize(Scale Factor=&Multiply_VectorByFloat(
     Vector=<1 1 1>
     Float=&FloatFromCurve(
      FloatCurve=DataInterface() {propertyValues {Curve {keys [
          {arriveTangent=4.671, interpMode=RCIM_Cubic, leaveTangent=4.671, tangentMode=RCTM_User, time=0, value=0.3}
          {arriveTangent=0.7287, interpMode=RCIM_Cubic, leaveTangent=0.7287, tangentMode=RCTM_User, time=0.2288, value=1.1761}
          {interpMode=RCIM_Cubic, tangentMode=RCTM_User, time=1, value=1.5}
         ]}}}
      Scale Curve=1.1
     )
    ))
   &ScaleMeshSize(Scale Factor=&MakeVector(
     X=&FloatFromCurve(FloatCurve=DataInterface() {propertyValues {Curve {keys [{arriveTangent=-4.9063, interpMode=RCIM_Cubic, leaveTangent=-4.9063, tangentMode=RCTM_User, time=0, value=3}, <at 1 0.8 RCIM_Cubic>]}}})
     Y=&FloatFromCurve(FloatCurve=DataInterface() {propertyValues {Curve {keys [<at 0 0.8>,<at 1 1.2>]}}})
     Z=&FloatFromCurve(FloatCurve=DataInterface() {propertyValues {Curve {keys [<at 0 0.8>,<at 1 1.2>]}}})
    )) {enabled=false, module=ScaleMeshSize001}
   /SetVariables(Particles.Scale002:Vector3f=&Multiply_Vector(
     A=<1 1 0.5>
     B=&VectorFromFloat(Value=&FloatFromCurve(
       FloatCurve=DataInterface() {propertyValues {Curve {keys [
           {arriveTangent=4.4162, interpMode=RCIM_Cubic, leaveTangent=4.4162, tangentMode=RCTM_User, time=0, value=0}
           {arriveTangent=1.3888, interpMode=RCIM_Cubic, leaveTangent=1.3888, tangentMode=RCTM_User, time=0.2558, value=1.1229}
           {arriveTangent=0.2529, interpMode=RCIM_Cubic, leaveTangent=0.2529, tangentMode=RCTM_User, time=1, value=1.65}
          ]}}}
       Scale Curve=0.8
      ))
    )) {enabled=false, isSetParameters=true}
   &ScaleColor(
    Scale Alpha=&FloatFromCurve(
     FloatCurve=DataInterface() {propertyValues {Curve {keys [
         {arriveTangent=-0.0176, interpMode=RCIM_Cubic, leaveTangent=-0.0176, tangentMode=RCTM_User, time=0, value=0}
         {arriveTangent=0.022, interpMode=RCIM_Cubic, leaveTangent=0.022, tangentMode=RCTM_User, time=0.2, value=1.425}
         {arriveTangent=-2.6564, interpMode=RCIM_Cubic, leaveTangent=-2.6564, tangentMode=RCTM_User, time=0.4, value=0.85}
         {arriveTangent=0.0267, tangentMode=RCTM_User, time=1, value=0}
        ]}}}
     Scale Curve=&Multiply_Float(A=2, B=&FNTimeOfDayScalar(Night Scale=0.5, Morning Scale=0.7, Evening Scale=0.5))
    )
    Scale RGB=&MakeVectorFromLinearColorRGB(Color=&ColorFromCurve(ColorCurve=DataInterface() {propertyValues {
        AlphaCurve {keys []}
        BlueCurve {keys [<at 0.0995 1.5 RCIM_Cubic>,<at 0.3925 1 RCIM_Cubic>]}
        GreenCurve {keys [<at 0.0995 1.5 RCIM_Cubic>,<at 0.3925 1 RCIM_Cubic>]}
        RedCurve {keys [<at 0.0995 1.5 RCIM_Cubic>,<at 0.3925 1 RCIM_Cubic>]}
       }}))
    ScaleA=true
    ScaleRGB=true
   )
   /SetVariables(Particles.Color002=&MakeLinearColorFromVectorAndFloat(
     "Vector (RGB)"=&MakeVectorFromLinearColorRGB(Color=<rgba 0.3281 0.1566 0.0237 1>)
     "Float (Alpha)"=&FloatFromCurve(
      FloatCurve=DataInterface() {propertyValues {Curve {keys [
          {arriveTangent=-3.4827, interpMode=RCIM_Cubic, leaveTangent=-3.4827, tangentMode=RCTM_User, time=0, value=1.5}
          {arriveTangent=-0.149, interpMode=RCIM_Cubic, leaveTangent=-0.149, tangentMode=RCTM_User, time=1, value=0}
         ]}}}
      Scale Curve=&Multiply_Float(A=1, B=&FNTimeOfDayScalar(Night Scale=0.5, Morning Scale=0.8, Evening Scale=0.7))
     )
    )) {enabled=false, isSetParameters=true}
   &SolveForcesAndVelocity(Clamp Velocity=false, Limit Acceleration=false)
  ]
  rendererClasses [&NiagaraMeshRendererProperties]
  renderers [
   {rendererClass=&NiagaraMeshRendererProperties, rendererIndex=0}
   {rendererClass=&NiagaraMeshRendererProperties, rendererIndex=1}
  ]
 }
 Smoke002 {
  emitterUpdateScript [
   &EmitterState(Inactive Response=::NewEnumerator1, Life Cycle Mode=::NewEnumerator1, Loop Behavior=::NewEnumerator1, Loop Duration=1, UseLoopDelay=false)
   &SpawnBurst_Instantaneous(Spawn Count=15, Use Spawn Probability=true, UseLoopCountLimit=false)
  ]
  particleSpawnScript [
   &InitializeParticle(
    Color Maximum=<rgba 0.0677 0.0622 0.0571 0.5>
    Color Minimum=<rgba 0.1 0.0861 0.0703 1>
    Color Mode=::NewEnumerator2
    Lifetime Max=1
    Lifetime Min=0.6
    Lifetime Mode=::NewEnumerator1
    Sprite Rotation Mode=::NewEnumerator2
    Sprite Size Mode=::NewEnumerator1
    Uniform Sprite Size Max=90
    Uniform Sprite Size Min=50
    UsePositionOffset=true
    Write Material Random=true
   )
   /SetVariables(Particles.InitSpriteRotRate=&UniformRangedFloat(Minimum=-3, Maximum=3)) {isSetParameters=true}
   &ShapeLocation(
    Handle Radius=10
    Large Radius=20
    MaskSpawnGroup=false
    Non Uniform Scale=<2 2 1>
    Offset Mode=::NewEnumerator0
    Override Seed=false
    Shape Primitive=::NewEnumerator3
    UseOwnerScale=true
   )
   &AddVelocity(
    Module.Use Falloff Distance=false
    Origin Offset=<0 0 -5>
    Override Seed=false
    Velocity Mode=::NewEnumerator1
    Velocity Speed=&RandomRangeFloat(Minimum=50, Maximum=200)
   )
   &AddVelocity(Velocity=<0 0 200>) {module=AddVelocity001}
   &CameraOffset(Camera Offset Amount=10)
  ]
  particleUpdateScript [
   &ParticleState()
   &SpriteRotationRate(Rotation Rate=&Multiply_Float(
     A=$Particles.InitSpriteRotRate
     B=&FloatFromCurve(
      FloatCurve=DataInterface() {propertyValues {Curve {keys [
          {arriveTangent=-4.0107, interpMode=RCIM_Cubic, leaveTangent=-4.0107, tangentMode=RCTM_User, time=-0.0046, value=1.0113}
          {arriveTangent=-0.2496, interpMode=RCIM_Cubic, leaveTangent=-0.2496, tangentMode=RCTM_User, time=0.5, value=0.16}
         ]}}}
      CurveIndex=$Emitter.Age
      Scale Curve=150
     )
    ))
   &ScaleColor(
    Scale Alpha=&MakeCustomFloatFromBool(
     Bool=$User.bIsFrontEndPreview
     True Float=0.5
     False Float=&FN_PickFloatDependOnPlatform(TargetPlatformActive=1, TargetPlatformNoActive=0.5)
    )
    Scale RGB=&MakeVectorFromLinearColorRGB(Color=&MakeCustomLinearColorFromBool(Bool=$User.bIsFrontEndPreview, True Color=<rgba 0.5 0.5 0.5 1>, False Color=<rgba 1 1 1 1>))
    ScaleA=true
    ScaleRGB=true
   ) {module=ScaleColor001}
   &ScaleColor(
    Scale Alpha=&FloatFromCurve(FloatCurve=DataInterface() {propertyValues {Curve {keys [<at 0 0 RCIM_Cubic>,<at 0.3 1 RCIM_Cubic>,<at 1 0 RCIM_Cubic>]}}})
    Scale RGB=&VectorFromFloat(Value=&FloatFromCurve(FloatCurve=DataInterface() {propertyValues {Curve {keys [
          {arriveTangent=1.8145, interpMode=RCIM_Cubic, leaveTangent=1.0283, tangentMode=RCTM_Break, time=0, value=0.4}
          {arriveTangent=0.4198, interpMode=RCIM_Cubic, leaveTangent=0.4198, tangentMode=RCTM_User, time=0.4, value=0.825}
          {arriveTangent=0.1918, tangentMode=RCTM_User, time=1, value=1}
         ]}}}))
    ScaleA=true
    ScaleRGB=true
   )
   /Niagara/Modules/Update/Forces/CurlNoiseForce(Noise Frequency=0.1, Noise Strength=200, Pan Noise=true, Pan Noise Field=<0 0 1>)
   &GravityForce(Gravity=<0 0 -500>)
   &Drag(Drag=2, Use Linear Drag=true, Use Rotational Drag=false)
   &SolveForcesAndVelocity(Clamp Velocity=false, Limit Acceleration=false)
   &ScaleSpriteSize(Scale Factor=&Vector2DFromFloat(Value=&FloatFromCurve(FloatCurve=DataInterface() {propertyValues {Curve {keys [
          {arriveTangent=8.1876, interpMode=RCIM_Cubic, leaveTangent=5.3162, tangentMode=RCTM_User, time=0, value=0.15}
          {arriveTangent=1.0025, interpMode=RCIM_Cubic, leaveTangent=1.0025, tangentMode=RCTM_User, time=0.166, value=0.7682}
          {arriveTangent=0.2754, interpMode=RCIM_Cubic, leaveTangent=0.2754, tangentMode=RCTM_User, time=1, value=1.2}
         ]}}})))
   &DynamicMaterialParameters(
    Index 0 Param 1=&Add_Float(
     A=$Particles.RandomFloat
     B=&FloatFromCurve(
      FloatCurve=DataInterface() {propertyValues {Curve {keys [
          {arriveTangent=2.7533, interpMode=RCIM_Cubic, leaveTangent=1.8958, tangentMode=RCTM_User, time=0, value=1}
          {arriveTangent=0.8292, interpMode=RCIM_Cubic, leaveTangent=0.8292, tangentMode=RCTM_User, time=0.321, value=1.3453}
          {arriveTangent=0.0529, interpMode=RCIM_Cubic, leaveTangent=0.2021, tangentMode=RCTM_User, time=1, value=1.5}
         ]}}}
      CurveIndex=&Add_Float(A=&Multiply_Float(A=$Particles.Initial.MaterialRandom, B=0.5), B=$Particles.NormalizedAge)
     )
    )
    Index 0 Param 2=&FloatFromCurve(FloatCurve=DataInterface() {propertyValues {Curve {keys [
         {arriveTangent=0.0453, interpMode=RCIM_Cubic, tangentMode=RCTM_Break, time=0, value=-1}
         {arriveTangent=0.0604, interpMode=RCIM_Cubic, leaveTangent=0.0604, tangentMode=RCTM_User, time=1, value=-1}
        ]}}})
    Index 0 Param 3=&FloatFromCurve(
     FloatCurve=DataInterface() {propertyValues {Curve {keys [
         {arriveTangent=-1.5217, interpMode=RCIM_Cubic, leaveTangent=-1.5217, tangentMode=RCTM_User, time=0, value=0.2}
         {arriveTangent=0.1017, interpMode=RCIM_Cubic, leaveTangent=0.2723, tangentMode=RCTM_Break, time=0.2, value=0}
        ]}}}
     Scale Curve=0.3
    )
    Index 0 Param 4=&FloatFromCurve(FloatCurve=DataInterface() {propertyValues {Curve {keys [<at 0.5 1>,<at 1 0>]}}})
    Index 1 Param 1=&FloatFromCurve(
     FloatCurve=DataInterface() {propertyValues {Curve {keys [
         {interpMode=RCIM_Cubic, leaveTangent=4.7088, tangentMode=RCTM_Break, time=0, value=0}
         {arriveTangent=4.2257, interpMode=RCIM_Cubic, leaveTangent=4.2257, tangentMode=RCTM_User, time=0.25, value=1}
        ]}}}
     Scale Curve=&RandomRangeFloat(Minimum=1.5, Maximum=1.5)
    )
    Index 1 Param 2=&FloatFromCurve(
     FloatCurve=DataInterface() {propertyValues {Curve {keys [<at 0 0.2>, {leaveTangent=0.9066, time=1, value=2}]}}}
     Scale Curve=&RandomRangeFloat(Minimum=0.2, Maximum=0.3)
    )
    Index 1 Param 3=&FloatFromCurve(FloatCurve=DataInterface() {propertyValues {Curve {keys [
         {interpMode=RCIM_Cubic, leaveTangent=6.3917, tangentMode=RCTM_Break, time=0, value=0.2}
         {arriveTangent=1.7549, interpMode=RCIM_Cubic, leaveTangent=10.6767, tangentMode=RCTM_User, time=0.25, value=1}
        ]}}})
    Index 1 Param 4=&FloatFromCurve(
     FloatCurve=DataInterface() {propertyValues {Curve {keys [{arriveTangent=-0.136, interpMode=RCIM_Cubic, leaveTangent=-0.136, tangentMode=RCTM_User, time=0, value=1}, <at 1 0.95>]}}}
     Scale Curve=&RandomRangeFloat(Minimum=0.5, Maximum=1)
    )
    Index 2 Param 1=&FloatFromCurve(FloatCurve=DataInterface() {propertyValues {Curve {keys [{interpMode=RCIM_Cubic, leaveTangent=1.1368, tangentMode=RCTM_Break, time=0, value=0.45}, <at 1 1>]}}})
    Index 2 Param 2=&FloatFromCurve(
     FloatCurve=DataInterface() {propertyValues {Curve {keys [
         {arriveTangent=1.9056, interpMode=RCIM_Cubic, leaveTangent=1.9056, tangentMode=RCTM_User, time=0, value=0.2}
         {arriveTangent=0.0259, interpMode=RCIM_Cubic, leaveTangent=0.0259, tangentMode=RCTM_User, time=1, value=1.4}
        ]}}}
     Scale Curve=&RandomRangeFloat(Minimum=0.2, Maximum=0.7)
    )
    Index 2 Param 3=&FloatFromCurve(
     FloatCurve=DataInterface() {propertyValues {Curve {keys [
         {interpMode=RCIM_Cubic, leaveTangent=2.3064, tangentMode=RCTM_Break, time=0, value=1}
         {interpMode=RCIM_Cubic, leaveTangent=11.1233, tangentMode=RCTM_Break, time=1, value=2}
        ]}}}
     Scale Curve=1.2
    )
    Index 2 Param 4=&FloatFromCurve(
     FloatCurve=DataInterface() {propertyValues {Curve {keys [
         {arriveTangent=-0.6867, interpMode=RCIM_Cubic, leaveTangent=-0.6867, tangentMode=RCTM_User, time=0, value=1}
         {interpMode=RCIM_Cubic, leaveTangent=-1.2956, tangentMode=RCTM_Break, time=1, value=0.75}
        ]}}}
     Scale Curve=1.2
    )
    Module.Param0WriteEnabled=true
    Module.Param0WriteEnabled1=true
    Module.Param0WriteEnabled2=true
    Module.Param1WriteEnabled=false
    Module.Param1WriteEnabled1=true
    Module.Param1WriteEnabled2=true
    Module.Param2WriteEnabled=true
    Module.Param2WriteEnabled1=true
    Module.Param2WriteEnabled2=true
    Module.Param3WriteEnabled=true
    Module.Param3WriteEnabled1=true
    Module.Param3WriteEnabled2=true
    Write Parameter Index 1=true
    Write Parameter Index 2=true
   )
  ]
  rendererClasses [&NiagaraSpriteRendererProperties]
  renderers [{rendererClass=&NiagaraSpriteRendererProperties}]
 }
 Soil {
  emitterUpdateScript [
   &EmitterState(Inactive Response=::NewEnumerator1, Life Cycle Mode=::NewEnumerator1, Loop Behavior=::NewEnumerator1, UseLoopDelay=false)
   &SpawnBurst_Instantaneous(Spawn Count=5, Use Spawn Probability=false, UseLoopCountLimit=false)
  ]
  particleSpawnScript [
   &InitializeParticle(
    Color Minimum=<rgba 0.2188 0.2188 0.2188 1>
    Color Mode=::NewEnumerator2
    Lifetime Max=0.5
    Lifetime Min=0.3
    Lifetime Mode=::NewEnumerator1
    Sprite Size Max=<100 130>
    Sprite Size Min=<80 100>
    Sprite Size Mode=::NewEnumerator2
    Sprite UV Mode=::NewEnumerator1
    UsePositionOffset=false
    Write Material Random=true
   )
   &CameraOffset(Camera Offset Amount=10) {enabled=false}
   &ShapeLocation(
    Handle Radius=5
    Large Radius=10
    MaskSpawnGroup=false
    Offset Mode=::NewEnumerator0
    Override Seed=false
    Rotation Axis=<0 0 1>
    Rotation Mode=::NewEnumerator1
    Shape Primitive=::NewEnumerator3
    UseOwnerScale=true
   )
   &AddVelocity(
    Module.Use Falloff Distance=false
    Origin Offset=<0 0 -5>
    Override Seed=false
    Velocity Mode=::NewEnumerator1
    Velocity Speed=&RandomRangeFloat(Minimum=20, Maximum=50)
   ) {module=AddVelocity002}
  ]
  particleUpdateScript [
   &ParticleState()
   &ScaleSpriteSize(Scale Sprite Size Mode=::NewEnumerator2) {enabled=false}
   &ScaleSpriteSize(
    Non-Uniform Curve Scale=<0.8 0.8>
    Non-Uniform Curve Sprite Scale=DataInterface() {propertyValues {
      XCurve {keys [
        {arriveTangent=0.9875, interpMode=RCIM_Cubic, leaveTangent=0.9875, tangentMode=RCTM_User, time=0, value=0.5}
        {arriveTangent=0.5103, interpMode=RCIM_Cubic, leaveTangent=0.5103, tangentMode=RCTM_User, time=0.3, value=0.8}
        {arriveTangent=0.0224, interpMode=RCIM_Cubic, leaveTangent=0.0224, tangentMode=RCTM_User, time=1, value=1}
       ]}
      YCurve {keys [
        {arriveTangent=0.8801, interpMode=RCIM_Cubic, leaveTangent=0.8801, tangentMode=RCTM_User, time=0, value=0.3}
        {arriveTangent=0.6586, interpMode=RCIM_Cubic, leaveTangent=0.6586, tangentMode=RCTM_User, time=0.2, value=0.85}
        {arriveTangent=-0.4778, interpMode=RCIM_Cubic, leaveTangent=-0.4778, tangentMode=RCTM_User, time=1, value=0.8}
       ]}
     }}
    Scale Sprite Size Mode=::NewEnumerator3
   ) {module=ScaleSpriteSize001}
   &ScaleColor(
    Scale Alpha=&FloatFromCurve(
     FloatCurve=DataInterface() {propertyValues {Curve {keys [
         <at 0 0 RCIM_Cubic>
         {arriveTangent=-0.0339, interpMode=RCIM_Cubic, leaveTangent=-0.0339, tangentMode=RCTM_User, time=0.1, value=1}
         {arriveTangent=-1.6214, interpMode=RCIM_Cubic, leaveTangent=-1.6214, tangentMode=RCTM_User, time=0.6616, value=0.2439}
         <at 1 0 RCIM_Cubic>
        ]}}}
     Scale Curve=&Multiply_Float(A=1, B=&FNTimeOfDayScalar(Night Scale=0.6, Morning Scale=0.8, Evening Scale=0.6))
    )
    Scale RGB=&MakeVectorFromLinearColorRGB(Color=&ColorFromCurve(ColorCurve=DataInterface() {propertyValues {
        AlphaCurve {keys []}
        BlueCurve {keys [<at 0.214 0.0713 RCIM_Cubic>,<at 0.7144 0.0019 RCIM_Cubic>]}
        GreenCurve {keys [<at 0.214 0.2252 RCIM_Cubic>,<at 0.7144 0.0168 RCIM_Cubic>]}
        RedCurve {keys [<at 0.214 0.3854 RCIM_Cubic>,<at 0.7144 0.04 RCIM_Cubic>]}
       }}))
    ScaleA=true
    ScaleRGB=true
   )
   &Drag(Drag=3, Use Linear Drag=true, Use Rotational Drag=true)
   &SolveForcesAndVelocity(Clamp Velocity=false, Limit Acceleration=false)
   &DynamicMaterialParameters(
    Index 0 Param 1=&FloatFromCurve(
     FloatCurve=DataInterface() {propertyValues {Curve {keys [{arriveTangent=-1.7327, interpMode=RCIM_Cubic, leaveTangent=-1.7327, tangentMode=RCTM_User, time=0, value=0.6}, <at 1 0 RCIM_Cubic>]}}}
     Scale Curve=&RandomRangeFloat(Minimum=0.5, Maximum=1)
    )
    Index 0 Param 2=&RandomRangeFloat(Minimum=0, Maximum=0.5)
    Index 0 Param 3=&RandomRangeFloat(Minimum=0, Maximum=0.5)
    Index 0 Param 4=&RandomRangeFloat(Minimum=1, Maximum=2)
    Module.Index0Float0Write=true
    Module.Index0Float1Write=true
    Module.Index0Float2Write=true
    Module.Index0Float3Write=true
   )
  ]
  rendererClasses [&NiagaraSpriteRendererProperties]
  renderers [{rendererClass=&NiagaraSpriteRendererProperties}]
 }
 Streak {
  emitterUpdateScript [
   &EmitterState(Inactive Response=::NewEnumerator1, Life Cycle Mode=::NewEnumerator1, Loop Behavior=::NewEnumerator1, Loop Duration=1, UseLoopDelay=false)
   &SpawnBurst_Instantaneous(Spawn Count=10, Use Spawn Probability=false, UseLoopCountLimit=false)
  ]
  particleSpawnScript [
   &InitializeParticle(
    Color Minimum=<rgba 0.1 0.1 0.1 2>
    Color Mode=::NewEnumerator2
    Lifetime Max=0.6
    Lifetime Min=0.3
    Lifetime Mode=::NewEnumerator1
    Sprite Size Max=<15 25>
    Sprite Size Min=<5 15>
    Sprite Size Mode=::NewEnumerator2
    Sprite UV Mode=::NewEnumerator1
    UsePositionOffset=false
    Write Material Random=true
   )
   &CameraOffset(Camera Offset Amount=10) {enabled=false}
   &ShapeLocation(
    MaskSpawnGroup=false
    Non Uniform Scale=<1 1 0.5>
    Offset Mode=::NewEnumerator0
    Override Seed=false
    Rotation Mode=::NewEnumerator1
    Sphere Radius=5
    UseOwnerScale=true
   )
   &AddVelocity(
    Cone Angle=160
    Cone Axis=<0 0 1>
    Module.UseConeFalloff=false
    Module.UseVelDistribution=false
    Override Seed=false
    Velocity Mode=::NewEnumerator2
    Velocity Speed=&RandomRangeFloat(Minimum=50, Maximum=100)
   ) {module=AddVelocity002}
  ]
  particleUpdateScript [
   &ParticleState()
   &ScaleSpriteSize(
    Scale Factor=&Multiply_Vector2D(
     A=&RandomRangeVector2D(Minimum=<1.3 2>, Maximum=<1.7 3>)
     B=&Vector2DFromFloat(Value=&FN_PickFloatDependOnPlatform(TargetPlatformActive=1, TargetPlatformNoActive=1.2))
    )
    Scale Sprite Size Mode=::NewEnumerator2
   )
   &ScaleSpriteSize(
    Non-Uniform Curve Scale=<1 2.3>
    Non-Uniform Curve Sprite Scale=DataInterface() {propertyValues {
      XCurve {keys [
        {arriveTangent=0.8609, interpMode=RCIM_Cubic, leaveTangent=0.8609, tangentMode=RCTM_User, time=0, value=0.6}
        {arriveTangent=0.5103, interpMode=RCIM_Cubic, leaveTangent=0.5103, tangentMode=RCTM_User, time=0.3, value=0.8}
        {arriveTangent=0.0224, interpMode=RCIM_Cubic, leaveTangent=0.0224, tangentMode=RCTM_User, time=1, value=1}
       ]}
      YCurve {keys [
        {arriveTangent=0.8801, interpMode=RCIM_Cubic, leaveTangent=0.8801, tangentMode=RCTM_User, time=0, value=0.3}
        {arriveTangent=0.6586, interpMode=RCIM_Cubic, leaveTangent=0.6586, tangentMode=RCTM_User, time=0.2, value=0.9}
        {arriveTangent=-0.4778, interpMode=RCIM_Cubic, leaveTangent=-0.4778, tangentMode=RCTM_User, time=1, value=0.8}
       ]}
     }}
    Scale Sprite Size Mode=::NewEnumerator3
   ) {module=ScaleSpriteSize001}
   &ScaleColor(
    Scale Alpha=&FloatFromCurve(
     FloatCurve=DataInterface() {propertyValues {Curve {keys [
         {arriveTangent=10.563, interpMode=RCIM_Cubic, leaveTangent=10.563, tangentMode=RCTM_User, time=0, value=0}
         {arriveTangent=-0.0296, interpMode=RCIM_Cubic, leaveTangent=-0.0296, tangentMode=RCTM_User, time=0.1, value=1}
         {arriveTangent=-2.1909, interpMode=RCIM_Cubic, leaveTangent=-2.1909, tangentMode=RCTM_User, time=0.5479, value=0.5472}
         <at 1 0 RCIM_Cubic>
        ]}}}
     Scale Curve=&Multiply_Float(A=1, B=&FNTimeOfDayScalar(Night Scale=0.6, Morning Scale=0.8, Evening Scale=0.6))
    )
    Scale RGB=&MakeVectorFromLinearColorRGB(Color=&ColorFromCurve(ColorCurve=DataInterface() {propertyValues {
        AlphaCurve {keys []}
        BlueCurve {keys [<at 0.1381 0.136 RCIM_Cubic>,<at 0.5196 0.0156 RCIM_Cubic>]}
        GreenCurve {keys [<at 0.1381 0.48 RCIM_Cubic>,<at 0.5196 0.0335 RCIM_Cubic>]}
        RedCurve {keys [<at 0.1381 0.8 RCIM_Cubic>,<at 0.5196 0.05 RCIM_Cubic>]}
       }}))
    ScaleA=true
    ScaleRGB=true
   )
   &Drag(Drag=2, Use Linear Drag=true, Use Rotational Drag=true)
   &SolveForcesAndVelocity(Clamp Velocity=false, Limit Acceleration=false)
   &DynamicMaterialParameters(
    Index 0 Param 1=&FloatFromCurve(FloatCurve=DataInterface() {propertyValues {Curve {keys [
         {arriveTangent=-0.1846, time=0, value=0}
         {arriveTangent=1.3385, interpMode=RCIM_Cubic, leaveTangent=1.1594, tangentMode=RCTM_Break, time=0.5, value=0}
         {arriveTangent=2, interpMode=RCIM_Cubic, leaveTangent=2, time=0.8, value=0.5}
         {arriveTangent=3.0091, interpMode=RCIM_Cubic, leaveTangent=3.0091, tangentMode=RCTM_User, time=1, value=1}
        ]}}})
    Index 0 Param 2=&FloatFromCurve(FloatCurve=DataInterface() {propertyValues {Curve {keys [
         {arriveTangent=-0.0189, interpMode=RCIM_Cubic, leaveTangent=-0.0189, tangentMode=RCTM_User, time=0, value=1}
         {arriveTangent=-3.722, interpMode=RCIM_Cubic, leaveTangent=-3.722, tangentMode=RCTM_User, time=0.8, value=0}
        ]}}})
    Index 0 Param 4=&FloatFromCurve(FloatCurve=DataInterface() {propertyValues {Curve {keys [<at 0 0 RCIM_Cubic>,<at 1 1 RCIM_Cubic>]}}}, Scale Curve=0.8)
    Module.Index0Param0Write=true
    Module.Index0Param1Write=true
    Module.Index0Param2Write=true
    Module.Index0Param3Write=true
   )
  ]
  rendererClasses [&NiagaraSpriteRendererProperties]
  renderers [{rendererClass=&NiagaraSpriteRendererProperties}]
 }
 Streaks_Dir {
  emitterUpdateScript [
   &EmitterState(Life Cycle Mode=::NewEnumerator1, Loop Behavior=::NewEnumerator1, Loop Delay=0.1187, Loop Duration=0.4, UseLoopDelay=false)
   &SpawnBurst_Instantaneous(Spawn Count=5, Use Spawn Probability=false, UseLoopCountLimit=false)
  ]
  particleSpawnScript [
   &PartitionParticles(Number Of Partitions=5)
   &InitializeParticle(
    Lifetime Max=0.6
    Lifetime Min=0.5
    Lifetime Mode=::NewEnumerator1
    Sprite Size=&SelectVector2DFromArray(
     Direct Array Index=$Particles.PartitionParticles.Partition
     Vector 2D Selection Array=DataInterface() {propertyValues {FloatData [<80 300>, <70 220>, <80 240>, <90 210>, <80 220>]}}
    )
    Sprite Size Mode=::NewEnumerator4
    UsePositionOffset=false
    Write Material Random=true
   )
   &ShapeLocation(
    MaskSpawnGroup=false
    Offset=&AddVectorToPosition(
     Position=&SelectVectorFromArray(
      Direct Array Index=$Particles.PartitionParticles.Partition
      Vector Selection Array=DataInterface() {propertyValues {FloatData [<0 0 -20>, <0 0 -25>, <5 5 -30>, <-5 -5 -25>, <-5 -5 -20>]}}
     )
     Vector=&MakeCustomVectorFromBool(Bool=$User.bIsFrontEndPreview, True Vector=<0 0 -10>, False Vector=<0 0 0>)
    )
    Offset Mode=::NewEnumerator0
    Override Seed=false
    Sphere Radius=1
    UseOwnerScale=true
   )
   &SubUVAnimation(End Frame=3, SubUV Animation Mode:ENiagara_SubUVLookupMode=::NewEnumerator2)
   &AddVelocity(Velocity=&SelectVectorFromArray(
     Direct Array Index=$Particles.PartitionParticles.Partition
     Vector Selection Array=DataInterface() {propertyValues {FloatData [<0 0 1>, <0.4 0 1>, <0 0.3 1>, <-0.3 0 1>, <0 -0.4 1>]}}
    ))
  ]
  particleUpdateScript [
   &ParticleState()
   &ScaleColor(
    Scale Alpha=&MakeCustomFloatFromBool(
     Bool=$User.bIsFrontEndPreview
     True Float=0.6
     False Float=&FN_PickFloatDependOnPlatform(TargetPlatformActive=1, TargetPlatformNoActive=0.5)
    )
    Scale RGB=&MakeCustomVectorFromBool(
     Bool=$User.bIsFrontEndPreview
     True Vector=&MakeVectorFromLinearColorRGB(Color=<rgba 0.5 0.5 0.5 1>)
     False Vector=&MakeVectorFromLinearColorRGB(Color=&FNTimeOfDayColor(Night Color=<rgba 0.401 0.401 0.401 1>, Morning Color=<rgba 0.6 0.6 0.6 1>, Day Color=<rgba 1 1 1 1>, Evening Color=<rgba 0.2 0.2 0.2 1>))
    )
    ScaleA=true
    ScaleRGB=true
   ) {module=ScaleColor001}
   &ScaleColor(
    Scale Alpha=&FloatFromCurve(
     FloatCurve=DataInterface() {propertyValues {Curve {keys [
         <at 0 0 RCIM_Cubic>
         {arriveTangent=-0.0456, interpMode=RCIM_Cubic, leaveTangent=-0.0456, tangentMode=RCTM_User, time=0.1, value=1.5}
         {arriveTangent=-2.8986, interpMode=RCIM_Cubic, leaveTangent=-2.8986, tangentMode=RCTM_User, time=0.3271, value=0.497}
         {arriveTangent=-0.8706, interpMode=RCIM_Cubic, leaveTangent=-0.8706, tangentMode=RCTM_User, time=0.5, value=0.175}
         <at 1 0 RCIM_Cubic>
        ]}}}
     Scale Curve=&RandomRangeFloat(Minimum=2, Maximum=3)
    )
    Scale RGB=&MakeVectorFromLinearColorRGB(Color=&ColorFromCurve(ColorCurve=DataInterface() {propertyValues {
        AlphaCurve {keys []}
        BlueCurve {keys [<at 0.1282 0.1375>,<at 0.7412 0.03>]}
        GreenCurve {keys [<at 0.1282 0.203>,<at 0.7412 0.03>]}
        RedCurve {keys [<at 0.1282 0.25>,<at 0.7412 0.03>]}
       }}))
    ScaleA=true
    ScaleRGB=true
   )
   &Drag(
    Drag=&FloatFromCurve(FloatCurve=DataInterface() {propertyValues {Curve {keys [<at 0.2 1>,<at 0.3 0>]}}}, Scale Curve=15)
    Use Linear Drag=true
    Use Rotational Drag=false
   )
   &SolveForcesAndVelocity(Clamp Velocity=false, Limit Acceleration=false)
   &ScaleSpriteSize(Scale Factor=&Vector2DFromCurve(
     Vector2Curve=DataInterface() {propertyValues {
       XCurve {keys [
         {arriveTangent=1.0909, interpMode=RCIM_Cubic, leaveTangent=1.0909, tangentMode=RCTM_User, time=0, value=0.25}
         {arriveTangent=0.7, interpMode=RCIM_Cubic, leaveTangent=0.7, time=0.3698, value=0.6052}
         {arriveTangent=0.2675, interpMode=RCIM_Cubic, leaveTangent=0.2675, tangentMode=RCTM_User, time=1, value=0.95}
        ]}
       YCurve {keys [
         {arriveTangent=4.701, interpMode=RCIM_Cubic, leaveTangent=4.701, tangentMode=RCTM_User, time=0, value=0}
         {arriveTangent=0.7056, interpMode=RCIM_Cubic, leaveTangent=0.7056, tangentMode=RCTM_User, time=0.1967, value=0.7589}
         <at 1 1 RCIM_Cubic>
        ]}
      }}
     ScaleCurve=<1 0.8>
    ))
   &DynamicMaterialParameters(
    Index 0 Param 1=&Multiply_Float(A=10, B=1)
    Index 0 Param 4=&FloatFromCurve(FloatCurve=DataInterface() {propertyValues {Curve {keys [{interpMode=RCIM_Cubic, tangentMode=RCTM_User, time=0, value=0}, {interpMode=RCIM_Cubic, tangentMode=RCTM_User, time=0.4, value=1}]}}})
    Index 1 Param 1=&FloatFromCurve(
     FloatCurve=DataInterface() {propertyValues {Curve {keys [
         {interpMode=RCIM_Cubic, tangentMode=RCTM_User, time=0, value=0}
         {arriveTangent=0.6295, interpMode=RCIM_Cubic, leaveTangent=0.6295, tangentMode=RCTM_User, time=1, value=1}
        ]}}}
     Scale Curve=0.42
    )
    Index 1 Param 2=&FloatFromCurve(
     FloatCurve=DataInterface() {propertyValues {Curve {keys [
         {arriveTangent=-100.8588, time=0, value=1}
         {arriveTangent=-26.7969, time=0.2, value=0.1}
         {arriveTangent=-2.6967, leaveTangent=-2.6967, time=0.3, value=0}
        ]}}}
     CurveIndex=$Emitter.Age
     Scale Curve=&MaskFloatBySpawnGroup(Masked Float=0, Passthrough Float=5, Spawn Group Mask=1)
    )
    Index 1 Param 3=&FloatFromCurve(
     FloatCurve=DataInterface() {propertyValues {Curve {keys [{interpMode=RCIM_Cubic, tangentMode=RCTM_User, time=0.2, value=0.5}, {interpMode=RCIM_Cubic, tangentMode=RCTM_User, time=1, value=1}]}}}
     Scale Curve=2
    )
    Index 1 Param 4=&FloatFromCurve(FloatCurve=DataInterface() {propertyValues {Curve {keys [<at 0 1>,<at 1 0>]}}}, Scale Curve=-0.3)
    Module.Param0WriteEnabled=true
    Module.Param0WriteEnabled1=true
    Module.Param1WriteEnabled=false
    Module.Param1WriteEnabled1=true
    Module.Param2WriteEnabled=false
    Module.Param2WriteEnabled1=true
    Module.Param3WriteEnabled=true
    Module.Param3WriteEnabled1=true
    Write Parameter Index 1=true
   )
  ]
  rendererClasses [&NiagaraSpriteRendererProperties]
  renderers [{rendererClass=&NiagaraSpriteRendererProperties}]
 }
 stone {
  emitterUpdateScript [
   &EmitterState(Inactive Response=::NewEnumerator1, Life Cycle Mode=::NewEnumerator1, Loop Behavior=::NewEnumerator1, Loop Duration=1, UseLoopDelay=false)
   &SpawnBurst_Instantaneous(Spawn Count=15, Use Spawn Probability=false, UseLoopCountLimit=false)
  ]
  particleSpawnScript [
   &InitializeParticle(
    Lifetime Max=1.5
    Lifetime Mode=::NewEnumerator1
    Sprite Rotation Mode=::NewEnumerator2
    Sprite Size Mode=::NewEnumerator1
    Uniform Sprite Size Min=3
    UsePositionOffset=false
    Write Material Random=true
   )
   &ShapeLocation(Hemisphere Distribution=<1 0.5>, MaskSpawnGroup=false, Override Seed=false, Sphere Radius=10, UseOwnerScale=true)
   &AddVelocity(
    Module.Use Falloff Distance=false
    Origin Offset=<0 0 -5>
    Override Seed=false
    Velocity Mode=::NewEnumerator1
    Velocity Speed=&RandomRangeFloat(Minimum=80, Maximum=220)
   )
   &AddVelocity(Velocity=&RandomRangeVector(Minimum=<0 0 150>, Maximum=<0 0 300>)) {module=AddVelocity001}
  ]
  particleUpdateScript [
   &ParticleState()
   &SpriteRotationRate(Rotation Rate=&Multiply_Float(A=&RandomRangeFloat(Minimum=-1, Maximum=1), B=800))
   &Drag(Drag=1.2, Use Linear Drag=true, Use Rotational Drag=true)
   &GravityForce()
   &SolveForcesAndVelocity(Clamp Velocity=false, Limit Acceleration=false)
   &ScaleColor(
    Linear Color Curve=DataInterface() {propertyValues {
      AlphaCurve {keys [<at 0 1>]}
      BlueCurve {keys [<at 0 0.1794>,<at 0.5 0.2448>]}
      GreenCurve {keys [<at 0 0.266>,<at 0.5 0.2448>]}
      RedCurve {keys [<at 0 0.3281>,<at 0.5 0.2448>]}
     }}
    Scale Mode=::NewEnumerator2
   )
  ]
  rendererClasses [&NiagaraSpriteRendererProperties]
  renderers [{rendererClass=&NiagaraSpriteRendererProperties}]
 }
}
summary {
 systemName=NS_Emote_SandalSite_ImpactBig
 emitters [
  {emitterName=Dirt, rendererClasses [&NiagaraSpriteRendererProperties]}
  {emitterName=Glow, rendererClasses [&NiagaraSpriteRendererProperties, &NiagaraLightRendererProperties]}
  {emitterName=Streak, rendererClasses [&NiagaraSpriteRendererProperties]}
  {emitterName=Ground_Ring, rendererClasses [&NiagaraMeshRendererProperties]}
  {emitterName=Soil, rendererClasses [&NiagaraSpriteRendererProperties]}
  {emitterName=stone, rendererClasses [&NiagaraSpriteRendererProperties]}
  {emitterName=Smoke002, rendererClasses [&NiagaraSpriteRendererProperties]}
  {emitterName=Streaks_Dir, rendererClasses [&NiagaraSpriteRendererProperties]}
 ]
 userVariables [
  User.bIsFrontEnd {type=NiagaraBool}
  User.bIsFrontEndPreview {type=NiagaraBool}
 ]
}
