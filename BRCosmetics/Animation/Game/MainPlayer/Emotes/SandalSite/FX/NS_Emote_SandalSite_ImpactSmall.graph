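# Niagara system for a small ground-impact effect in the SandalSite emote, built from five independent burst emitters.
# Glow flashes a warm sprite with a particle light, while Streak and Soil spray short-lived, darkening sprites.
# Stone throws small spinning debris under gravity, and Smoke002 adds curl-noise puffs driven by dynamic material parameters.
# Brightness, size and opacity adapt in places to time of day, front-end preview and platform.
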NiagaraSystem /BRCosmetics/Animation/Game/MainPlayer/Emotes/SandalSite/FX/NS_Emote_SandalSite_ImpactSmall
dependencies {
 usedDataInterfaces [
  &NiagaraDataInterfaceCurve
  &NiagaraDataInterfaceColorCurve
  &NiagaraDataInterfaceArrayFloat
  &NiagaraDataInterfaceVector2DCurve
  &NiagaraDataInterfacePlatformSet
 ]
 usedDynamicInputs [&Lerp_Float]
 usedModules [&SystemState]
}
emitters {
 Glow {
  emitterUpdateScript [
   &EmitterState(
    Inactive Response=::NewEnumerator1
    Life Cycle Mode=::NewEnumerator1
    Loop Behavior=::NewEnumerator1
    Loop Delay=1.6
    Loop Duration=0.5
    UseLoopDelay=false
   )
   &SpawnBurst_Instantaneous(Use Spawn Probability=false, UseLoopCountLimit=false)
  ]
  particleSpawnScript [
   &InitializeParticle(
    Color=<rgba 1 0.5485 0.0896 1>
    Lifetime=0.3
    Sprite Rotation Mode=::NewEnumerator2
    Sprite Size Mode=::NewEnumerator3
    Uniform Sprite Size=120
    UsePositionOffset=false
    Write Material Random=true
   )
   &SystemLocation(MaskSpawnGroup=false, Offset=<0 0 50>)
   &CameraOffset(Camera Offset Amount=50)
   /SetVariables() {isSetParameters=true}
   /SetVariables(Particles.Position002:NiagaraPosition=<0 0 5>) {isSetParameters=true}
  ]
  particleUpdateScript [
   &ParticleState()
   &ScaleSpriteSize(
    Uniform Curve Scale=&FN_FrontEndFloatModifier(Gameplay=1, Front End=0.8)
    Uniform Curve Sprite Scale=DataInterface() {propertyValues {Curve {keys [
        {arriveTangent=-0.7716, interpMode=RCIM_Cubic, leaveTangent=-0.7716, tangentMode=RCTM_User, time=0, value=0}
        {arriveTangent=-5.3534, interpMode=RCIM_Cubic, leaveTangent=-5.3534, tangentMode=RCTM_User, time=0.2, value=2.5}
        {arriveTangent=-1.1177, interpMode=RCIM_Cubic, leaveTangent=-1.1177, tangentMode=RCTM_User, time=0.3, value=1.375}
        {arriveTangent=-0.2284, interpMode=RCIM_Cubic, leaveTangent=-0.2284, tangentMode=RCTM_User, time=1, value=0.875}
       ]}}}
   )
   /SetVariables(Particles.Size002:Vector2f=&Multiply_Vector2D(A=$Particles.SpriteSize, B=<0.5 0.5>)) {isSetParameters=true}
   &ScaleColor(
    Scale Alpha=&FloatFromCurve(
     FloatCurve=DataInterface() {propertyValues {Curve {keys [
         {arriveTangent=-2.7027, interpMode=RCIM_Cubic, leaveTangent=-2.7027, tangentMode=RCTM_User, time=0, value=1}
         {arriveTangent=-1, interpMode=RCIM_Cubic, leaveTangent=-1, time=0.312, value=0.3309}
         <at 1 0>
        ]}}}
     Scale Curve=&Multiply_Float(A=1, B=&FNTimeOfDayScalar(Night Scale=0.8, Morning Scale=1, Evening Scale=1))
    )
    ScaleA=true
    ScaleRGB=true
   )
   /SetVariables(Particles.Color002=&MakeLinearColorFromVectorAndFloat(
     "Vector (RGB)"=&MakeVectorFromLinearColorRGB(Color=<rgba 1 0.5 0.0896 1>)
     "Float (Alpha)"=&FloatFromCurve(
      FloatCurve=DataInterface() {propertyValues {Curve {keys [
          {arriveTangent=-2.7027, interpMode=RCIM_Cubic, leaveTangent=-2.7027, tangentMode=RCTM_User, time=0, value=1}
          {arriveTangent=-0.5419, interpMode=RCIM_Cubic, leaveTangent=-0.5419, tangentMode=RCTM_User, time=0.5, value=0.1}
          <at 1 0>
         ]}}}
      Scale Curve=&Multiply_Float(A=0.5, B=&FNTimeOfDayScalar(Night Scale=0.8, Morning Scale=1, Evening Scale=1))
     )
    )) {isSetParameters=true}
   &FN_Light_Attributes(
    Color=&ColorFromCurve(ColorCurve=DataInterface() {propertyValues {
       AlphaCurve {keys []}
       BlueCurve {keys [<at 0.2351 0.4792 RCIM_Cubic>,<at 0.5132 0.1719>]}
       GreenCurve {keys [<at 0.2351 1.2113 RCIM_Cubic>,<at 0.5132 0.5705>]}
       RedCurve {keys [<at 0.2351 2 RCIM_Cubic>,<at 0.5132 1>]}
      }})
    Create Color Attribute=true
    Create Exponent Attribute=true
    Create Position Attribute=true
    Create Radius Attribute=true
    Enable Particle Light Enable=true
    Exponent=2
    Position=<0 0 50>
    Radius=&Multiply_Float(
     A=40
     B=&FloatFromCurve(FloatCurve=DataInterface() {propertyValues {Curve {keys [<at 0 1.8>,<at 0.2 2.05>,<at 0.5 2.2>,<at 1 1.8>]}}})
    )
    Scale Color=&FloatFromCurve(
     FloatCurve=DataInterface() {propertyValues {Curve {keys [
         {arriveTangent=5.626, interpMode=RCIM_Cubic, leaveTangent=5.626, tangentMode=RCTM_User, time=0, value=0.6}
         {arriveTangent=0.0895, interpMode=RCIM_Cubic, leaveTangent=0.0895, tangentMode=RCTM_User, time=0.1, value=1}
         {arriveTangent=-0.8586, interpMode=RCIM_Cubic, leaveTangent=-0.8586, tangentMode=RCTM_User, time=0.815, value=0.72}
         {arriveTangent=-4.4162, interpMode=RCIM_Cubic, leaveTangent=-4.4162, tangentMode=RCTM_User, time=1, value=0}
        ]}}}
     Scale Curve=&Multiply_Float(
      A=&FN_FrontEndFloatModifier(Gameplay=25, Front End=5)
      B=&FNTimeOfDayScalar(Night Scale=0.5, Morning Scale=1, Evening Scale=0.8)
     )
    )
   )
  ]
  rendererClasses [&NiagaraSpriteRendererProperties, &NiagaraLightRendererProperties]
  renderers [
   {rendererClass=&NiagaraSpriteRendererProperties}
   {rendererClass=&NiagaraLightRendererProperties, rendererIndex=1}
   {rendererClass=&NiagaraSpriteRendererProperties, rendererIndex=2}
  ]
 }
 Smoke002 {
  emitterUpdateScript [
   &EmitterState(Inactive Response=::NewEnumerator1, Life Cycle Mode=::NewEnumerator1, Loop Behavior=::NewEnumerator1, Loop Duration=1, UseLoopDelay=false)
   &SpawnBurst_Instantaneous(Spawn Count=6, Use Spawn Probability=true, UseLoopCountLimit=false)
  ]
  particleSpawnScript [
   &InitializeParticle(
    Color Maximum=<rgba 0.0885 0.0814 0.0747 0.5>
    Color Minimum=<rgba 0.1 0.0861 0.0703 1>
    Color Mode=::NewEnumerator2
    Lifetime Max=0.6
    Lifetime Min=0.5
    Lifetime Mode=::NewEnumerator1
    Sprite Rotation Mode=::NewEnumerator2
    Sprite Size Mode=::NewEnumerator1
    Uniform Sprite Size Max=60
    Uniform Sprite Size Min=30
    UsePositionOffset=true
    Write Material Random=true
   )
   /SetVariables(Particles.InitSpriteRotRate=&UniformRangedFloat(Minimum=-3, Maximum=3)) {isSetParameters=true}
   &ShapeLocation(
    Handle Radius=5
    Large Radius=5
    MaskSpawnGroup=false
    Non Uniform Scale=<2 2 1>
    Offset Mode=::NewEnumerator0
    Override Seed=false
    Shape Primitive=::NewEnumerator3
    UseOwnerScale=true
   )
   &AddVelocity(
    Module.Use Falloff Distance=false
    Origin Offset=<0 0 -5>
    Override Seed=false
    Velocity Mode=::NewEnumerator1
    Velocity Speed=&RandomRangeFloat(Minimum=50, Maximum=80)
   )
   &AddVelocity() {module=AddVelocity001}
   &CameraOffset(Camera Offset Amount=10)
  ]
  particleUpdateScript [
   &ParticleState()
   &SpriteRotationRate(Rotation Rate=&Multiply_Float(
     A=$Particles.InitSpriteRotRate
     B=&FloatFromCurve(
      FloatCurve=DataInterface() {propertyValues {Curve {keys [
          {arriveTangent=-4.0107, interpMode=RCIM_Cubic, leaveTangent=-4.0107, tangentMode=RCTM_User, time=-0.0046, value=1.0113}
          {arriveTangent=-0.2496, interpMode=RCIM_Cubic, leaveTangent=-0.2496, tangentMode=RCTM_User, time=0.5, value=0.16}
         ]}}}
      CurveIndex=$Emitter.Age
      Scale Curve=150
     )
    ))
   &ScaleColor(
    Scale Alpha=&MakeCustomFloatFromBool(
     Bool=$User.bIsFrontEndPreview
     True Float=0.5
     False Float=&FN_PickFloatDependOnPlatform(TargetPlatformActive=1, TargetPlatformNoActive=0.5)
    )
    Scale RGB=&MakeVectorFromLinearColorRGB(Color=&MakeCustomLinearColorFromBool(Bool=$User.bIsFrontEndPreview, True Color=<rgba 0.5 0.5 0.5 1>, False Color=<rgba 1 1 1 1>))
    ScaleA=true
    ScaleRGB=true
   ) {module=ScaleColor001}
   &ScaleColor(
    Scale Alpha=&FloatFromCurve(FloatCurve=DataInterface() {propertyValues {Curve {keys [<at 0 0 RCIM_Cubic>,<at 0.3 1 RCIM_Cubic>,<at 1 0 RCIM_Cubic>]}}})
    Scale RGB=&VectorFromFloat(Value=&FloatFromCurve(
      FloatCurve=DataInterface() {propertyValues {Curve {keys [
          {arriveTangent=1.8145, interpMode=RCIM_Cubic, leaveTangent=1.0283, tangentMode=RCTM_Break, time=0, value=0.4}
          {arriveTangent=0.4198, interpMode=RCIM_Cubic, leaveTangent=0.4198, tangentMode=RCTM_User, time=0.4, value=0.825}
          {arriveTangent=0.1918, tangentMode=RCTM_User, time=1, value=1}
         ]}}}
      Scale Curve=&FN_PickFloatDependOnPlatform(TargetPlatformActive=1, TargetPlatformNoActive=0.8)
     ))
    ScaleA=true
    ScaleRGB=true
   )
   /Niagara/Modules/Update/Forces/CurlNoiseForce(Noise Frequency=0.1, Noise Strength=200, Pan Noise=true, Pan Noise Field=<0 0 1>)
   &GravityForce(Gravity=<0 0 -300>)
   &Drag(Drag=2, Use Linear Drag=true, Use Rotational Drag=false)
   &SolveForcesAndVelocity(Clamp Velocity=false, Limit Acceleration=false)
   &ScaleSpriteSize(Scale Factor=&Vector2DFromFloat(Value=&FloatFromCurve(FloatCurve=DataInterface() {propertyValues {Curve {keys [
          {arriveTangent=8.1876, interpMode=RCIM_Cubic, leaveTangent=5.3162, tangentMode=RCTM_User, time=0, value=0.15}
          {arriveTangent=1.0025, interpMode=RCIM_Cubic, leaveTangent=1.0025, tangentMode=RCTM_User, time=0.166, value=0.7682}
          {arriveTangent=0.2754, interpMode=RCIM_Cubic, leaveTangent=0.2754, tangentMode=RCTM_User, time=1, value=1.2}
         ]}}})))
   &DynamicMaterialParameters(
    Index 0 Param 1=&Add_Float(
     A=$Particles.RandomFloat
     B=&FloatFromCurve(
      FloatCurve=DataInterface() {propertyValues {Curve {keys [
          {arriveTangent=2.7533, interpMode=RCIM_Cubic, leaveTangent=1.8958, tangentMode=RCTM_User, time=0, value=1}
          {arriveTangent=0.8292, interpMode=RCIM_Cubic, leaveTangent=0.8292, tangentMode=RCTM_User, time=0.321, value=1.3453}
          {arriveTangent=0.0529, interpMode=RCIM_Cubic, leaveTangent=0.2021, tangentMode=RCTM_User, time=1, value=1.5}
         ]}}}
      CurveIndex=&Add_Float(A=&Multiply_Float(A=$Particles.Initial.MaterialRandom, B=0.5), B=$Particles.NormalizedAge)
     )
    )
    Index 0 Param 2=&FloatFromCurve(FloatCurve=DataInterface() {propertyValues {Curve {keys [
         {arriveTangent=0.0453, interpMode=RCIM_Cubic, tangentMode=RCTM_Break, time=0, value=-1}
         {arriveTangent=0.0604, interpMode=RCIM_Cubic, leaveTangent=0.0604, tangentMode=RCTM_User, time=1, value=-1}
        ]}}})
    Index 0 Param 3=&FloatFromCurve(
     FloatCurve=DataInterface() {propertyValues {Curve {keys [
         {arriveTangent=-1.5217, interpMode=RCIM_Cubic, leaveTangent=-1.5217, tangentMode=RCTM_User, time=0, value=0.2}
         {arriveTangent=0.1017, interpMode=RCIM_Cubic, leaveTangent=0.2723, tangentMode=RCTM_Break, time=0.2, value=0}
        ]}}}
     Scale Curve=0.3
    )
    Index 0 Param 4=&FloatFromCurve(FloatCurve=DataInterface() {propertyValues {Curve {keys [<at 0.5 1>,<at 1 0>]}}})
    Index 1 Param 1=&FloatFromCurve(
     FloatCurve=DataInterface() {propertyValues {Curve {keys [
         {interpMode=RCIM_Cubic, leaveTangent=4.7088, tangentMode=RCTM_Break, time=0, value=0}
         {arriveTangent=4.2257, interpMode=RCIM_Cubic, leaveTangent=4.2257, tangentMode=RCTM_User, time=0.25, value=1}
        ]}}}
     Scale Curve=&RandomRangeFloat(Minimum=1.5, Maximum=1.5)
    )
    Index 1 Param 2=&FloatFromCurve(
     FloatCurve=DataInterface() {propertyValues {Curve {keys [<at 0 0.2>, {leaveTangent=0.9066, time=1, value=2}]}}}
     Scale Curve=&RandomRangeFloat(Minimum=0.2, Maximum=0.3)
    )
    Index 1 Param 3=&FloatFromCurve(FloatCurve=DataInterface() {propertyValues {Curve {keys [
         {interpMode=RCIM_Cubic, leaveTangent=6.3917, tangentMode=RCTM_Break, time=0, value=0.2}
         {arriveTangent=1.7549, interpMode=RCIM_Cubic, leaveTangent=10.6767, tangentMode=RCTM_User, time=0.25, value=1}
        ]}}})
    Index 1 Param 4=&FloatFromCurve(
     FloatCurve=DataInterface() {propertyValues {Curve {keys [{arriveTangent=-0.136, interpMode=RCIM_Cubic, leaveTangent=-0.136, tangentMode=RCTM_User, time=0, value=1}, <at 1 0.95>]}}}
     Scale Curve=&RandomRangeFloat(Minimum=0.5, Maximum=1)
    )
    Index 2 Param 1=&FloatFromCurve(FloatCurve=DataInterface() {propertyValues {Curve {keys [{interpMode=RCIM_Cubic, leaveTangent=1.1368, tangentMode=RCTM_Break, time=0, value=0.45}, <at 1 1>]}}})
    Index 2 Param 2=&FloatFromCurve(
     FloatCurve=DataInterface() {propertyValues {Curve {keys [
         {arriveTangent=1.9056, interpMode=RCIM_Cubic, leaveTangent=1.9056, tangentMode=RCTM_User, time=0, value=0.2}
         {arriveTangent=0.0259, interpMode=RCIM_Cubic, leaveTangent=0.0259, tangentMode=RCTM_User, time=1, value=1.4}
        ]}}}
     Scale Curve=&RandomRangeFloat(Minimum=0.2, Maximum=0.7)
    )
    Index 2 Param 3=&FloatFromCurve(
     FloatCurve=DataInterface() {propertyValues {Curve {keys [
         {interpMode=RCIM_Cubic, leaveTangent=2.3064, tangentMode=RCTM_Break, time=0, value=1}
         {interpMode=RCIM_Cubic, leaveTangent=11.1233, tangentMode=RCTM_Break, time=1, value=2}
        ]}}}
     Scale Curve=1.2
    )
    Index 2 Param 4=&FloatFromCurve(
     FloatCurve=DataInterface() {propertyValues {Curve {keys [
         {arriveTangent=-0.6867, interpMode=RCIM_Cubic, leaveTangent=-0.6867, tangentMode=RCTM_User, time=0, value=1}
         {interpMode=RCIM_Cubic, leaveTangent=-1.2956, tangentMode=RCTM_Break, time=1, value=0.75}
        ]}}}
     Scale Curve=1.2
    )
    Module.Param0WriteEnabled=true
    Module.Param0WriteEnabled1=true
    Module.Param0WriteEnabled2=true
    Module.Param1WriteEnabled=false
    Module.Param1WriteEnabled1=true
    Module.Param1WriteEnabled2=true
    Module.Param2WriteEnabled=true
    Module.Param2WriteEnabled1=true
    Module.Param2WriteEnabled2=true
    Module.Param3WriteEnabled=true
    Module.Param3WriteEnabled1=true
    Module.Param3WriteEnabled2=true
    Write Parameter Index 1=true
    Write Parameter Index 2=true
   )
  ]
  rendererClasses [&NiagaraSpriteRendererProperties]
  renderers [{rendererClass=&NiagaraSpriteRendererProperties}]
 }
 Soil {
  emitterUpdateScript [
   &EmitterState(Inactive Response=::NewEnumerator1, Life Cycle Mode=::NewEnumerator1, Loop Behavior=::NewEnumerator1, UseLoopDelay=false)
   &SpawnBurst_Instantaneous(Spawn Count=2, Use Spawn Probability=false, UseLoopCountLimit=false)
  ]
  particleSpawnScript [
   &InitializeParticle(
    Color Minimum=<rgba 0.2188 0.2188 0.2188 1>
    Color Mode=::NewEnumerator2
    Lifetime Max=0.5
    Lifetime Min=0.3
    Lifetime Mode=::NewEnumerator1
    Sprite Size Max=<80 120>
    Sprite Size Min=<60 80>
    Sprite Size Mode=::NewEnumerator2
    Sprite UV Mode=::NewEnumerator1
    UsePositionOffset=false
    Write Material Random=true
   )
   &CameraOffset(Camera Offset Amount=&SelectFloatFromArray(Direct Array Index=&SpawnGroup(), Float Selection Array=DataInterface() {propertyValues {FloatData [10,10]}})) {enabled=false}
   &ShapeLocation(MaskSpawnGroup=false, Offset Mode=::NewEnumerator0, Override Seed=false, Rotation Mode=::NewEnumerator1, Sphere Radius=1, UseOwnerScale=true)
   &AddVelocity(
    Cone Angle=160
    Cone Axis=<0 0 1>
    Module.UseConeFalloff=false
    Module.UseVelDistribution=false
    Override Seed=false
    Velocity Mode=::NewEnumerator2
    Velocity Speed=&RandomRangeFloat(Minimum=20, Maximum=30)
   ) {module=AddVelocity002}
  ]
  particleUpdateScript [
   &ParticleState()
   &ScaleSpriteSize(
    Non-Uniform Curve Scale=<0.6 0.5>
    Non-Uniform Curve Sprite Scale=DataInterface() {propertyValues {
      XCurve {keys [
        {arriveTangent=0.9875, interpMode=RCIM_Cubic, leaveTangent=0.9875, tangentMode=RCTM_User, time=0, value=0.5}
        {arriveTangent=0.5103, interpMode=RCIM_Cubic, leaveTangent=0.5103, tangentMode=RCTM_User, time=0.3, value=0.8}
        {arriveTangent=0.0224, interpMode=RCIM_Cubic, leaveTangent=0.0224, tangentMode=RCTM_User, time=1, value=1}
       ]}
      YCurve {keys [
        {arriveTangent=0.8801, interpMode=RCIM_Cubic, leaveTangent=0.8801, tangentMode=RCTM_User, time=0, value=0.3}
        {arriveTangent=0.6586, interpMode=RCIM_Cubic, leaveTangent=0.6586, tangentMode=RCTM_User, time=0.2, value=0.85}
        {arriveTangent=-0.4778, interpMode=RCIM_Cubic, leaveTangent=-0.4778, tangentMode=RCTM_User, time=1, value=0.8}
       ]}
     }}
    Scale Sprite Size Mode=::NewEnumerator3
   ) {module=ScaleSpriteSize001}
   &ScaleColor(
    Scale Alpha=&FloatFromCurve(
     FloatCurve=DataInterface() {propertyValues {Curve {keys [
         <at 0 0 RCIM_Cubic>
         {arriveTangent=-0.0339, interpMode=RCIM_Cubic, leaveTangent=-0.0339, tangentMode=RCTM_User, time=0.1, value=1}
         {arriveTangent=-1.6214, interpMode=RCIM_Cubic, leaveTangent=-1.6214, tangentMode=RCTM_User, time=0.6616, value=0.2439}
         <at 1 0 RCIM_Cubic>
        ]}}}
     Scale Curve=&Multiply_Float(A=1, B=&FNTimeOfDayScalar(Night Scale=0.6, Morning Scale=0.8, Evening Scale=0.6))
    )
    Scale RGB=&MakeVectorFromLinearColorRGB(Color=&ColorFromCurve(ColorCurve=DataInterface() {propertyValues {
        AlphaCurve {keys []}
        BlueCurve {keys [<at 0.0947 0.1587 RCIM_Cubic>,<at 0.6746 0.0148 RCIM_Cubic>]}
        GreenCurve {keys [<at 0.0947 0.2775 RCIM_Cubic>,<at 0.6746 0.0246 RCIM_Cubic>]}
        RedCurve {keys [<at 0.0947 0.401 RCIM_Cubic>,<at 0.6746 0.04 RCIM_Cubic>]}
       }}))
    ScaleA=true
    ScaleRGB=true
   )
   &Drag(Drag=3, Use Linear Drag=true, Use Rotational Drag=true)
   &SolveForcesAndVelocity(Clamp Velocity=false, Limit Acceleration=false)
   &DynamicMaterialParameters(
    Index 0 Param 1=&FloatFromCurve(
     FloatCurve=DataInterface() {propertyValues {Curve {keys [{arriveTangent=-1.447, interpMode=RCIM_Cubic, leaveTangent=-1.447, tangentMode=RCTM_User, time=0, value=0.6}, <at 1 0 RCIM_Cubic>]}}}
     Scale Curve=&RandomRangeFloat(Minimum=0.5, Maximum=1)
    )
    Index 0 Param 2=&RandomRangeFloat(Minimum=0, Maximum=0.5)
    Index 0 Param 3=&RandomRangeFloat(Minimum=0, Maximum=0.5)
    Index 0 Param 4=&RandomRangeFloat(Minimum=1, Maximum=2)
    Module.Index0Float0Write=true
    Module.Index0Float1Write=true
    Module.Index0Float2Write=true
    Module.Index0Float3Write=true
   )
  ]
  rendererClasses [&NiagaraSpriteRendererProperties]
  renderers [{rendererClass=&NiagaraSpriteRendererProperties}]
 }
 Streak {
  emitterUpdateScript [
   &EmitterState(Inactive Response=::NewEnumerator1, Life Cycle Mode=::NewEnumerator1, Loop Behavior=::NewEnumerator1, Loop Duration=1, UseLoopDelay=false)
   &SpawnBurst_Instantaneous(Spawn Count=8, Use Spawn Probability=false, UseLoopCountLimit=false)
  ]
  particleSpawnScript [
   &InitializeParticle(
    Color Minimum=<rgba 0.4583 0.4583 0.4583 2>
    Color Mode=::NewEnumerator2
    Lifetime Max=0.5
    Lifetime Min=0.3
    Lifetime Mode=::NewEnumerator1
    Sprite Size Max=<15 50>
    Sprite Size Min=<5 30>
    Sprite Size Mode=::NewEnumerator2
    Sprite UV Mode=::NewEnumerator1
    UsePositionOffset=false
    Write Material Random=true
   )
   &CameraOffset(Camera Offset Amount=10) {enabled=false}
   &ShapeLocation(
    Handle Radius=2
    Large Radius=2
    MaskSpawnGroup=false
    Offset Mode=::NewEnumerator0
    Override Seed=false
    Rotation Mode=::NewEnumerator1
    Shape Primitive=::NewEnumerator3
    UseOwnerScale=true
   )
   &AddVelocity(
    Cone Angle=160
    Cone Axis=<0 0 1>
    Module.UseConeFalloff=false
    Module.UseVelDistribution=false
    Override Seed=false
    Velocity Mode=::NewEnumerator2
    Velocity Speed=&RandomRangeFloat(Minimum=50, Maximum=100)
   ) {module=AddVelocity002}
  ]
  particleUpdateScript [
   &ParticleState()
   &ScaleSpriteSize(
    Non-Uniform Curve Scale=<0.8 1.5>
    Non-Uniform Curve Sprite Scale=DataInterface() {propertyValues {
      XCurve {keys [
        {arriveTangent=0.8609, interpMode=RCIM_Cubic, leaveTangent=0.8609, tangentMode=RCTM_User, time=0, value=0.6}
        {arriveTangent=0.5103, interpMode=RCIM_Cubic, leaveTangent=0.5103, tangentMode=RCTM_User, time=0.3, value=0.8}
        {arriveTangent=0.0224, interpMode=RCIM_Cubic, leaveTangent=0.0224, tangentMode=RCTM_User, time=1, value=1}
       ]}
      YCurve {keys [
        {arriveTangent=0.8801, interpMode=RCIM_Cubic, leaveTangent=0.8801, tangentMode=RCTM_User, time=0, value=0.3}
        {arriveTangent=0.6586, interpMode=RCIM_Cubic, leaveTangent=0.6586, tangentMode=RCTM_User, time=0.2, value=0.85}
        {arriveTangent=-0.4778, interpMode=RCIM_Cubic, leaveTangent=-0.4778, tangentMode=RCTM_User, time=1, value=0.75}
       ]}
     }}
    Scale Sprite Size Mode=::NewEnumerator3
   ) {module=ScaleSpriteSize001}
   &ScaleColor(
    Scale Alpha=&FloatFromCurve(
     FloatCurve=DataInterface() {propertyValues {Curve {keys [
         <at 0 0 RCIM_Cubic>
         {arriveTangent=-1.2233, interpMode=RCIM_Cubic, leaveTangent=-1.2233, tangentMode=RCTM_User, time=0.1, value=1}
         <at 1 0 RCIM_Cubic>
        ]}}}
     Scale Curve=&Multiply_Float(A=1, B=&FNTimeOfDayScalar(Night Scale=0.6, Morning Scale=0.8, Evening Scale=0.6))
    )
    Scale RGB=&MakeVectorFromLinearColorRGB(Color=&ColorFromCurve(ColorCurve=DataInterface() {propertyValues {
        AlphaCurve {keys []}
        BlueCurve {keys [<at 0.0811 0.1501 RCIM_Cubic>,<at 0.5555 0.0049 RCIM_Cubic>]}
        GreenCurve {keys [<at 0.0811 0.3307 RCIM_Cubic>,<at 0.5555 0.0372 RCIM_Cubic>]}
        RedCurve {keys [<at 0.0811 0.5 RCIM_Cubic>,<at 0.5555 0.0573 RCIM_Cubic>]}
       }}))
    ScaleA=true
    ScaleRGB=true
   )
   &Drag(Drag=2.5, Use Linear Drag=true, Use Rotational Drag=true)
   &SolveForcesAndVelocity(Clamp Velocity=false, Limit Acceleration=false)
   &DynamicMaterialParameters(
    Index 0 Param 1=&FloatFromCurve(FloatCurve=DataInterface() {propertyValues {Curve {keys [
         {arriveTangent=-0.1846, time=0, value=0}
         {arriveTangent=1.3385, interpMode=RCIM_Cubic, leaveTangent=1.1594, tangentMode=RCTM_Break, time=0.5, value=0}
         {arriveTangent=2, interpMode=RCIM_Cubic, leaveTangent=2, time=0.8, value=0.5}
         {arriveTangent=3.0091, interpMode=RCIM_Cubic, leaveTangent=3.0091, tangentMode=RCTM_User, time=1, value=1}
        ]}}})
    Index 0 Param 2=&FloatFromCurve(FloatCurve=DataInterface() {propertyValues {Curve {keys [
         {arriveTangent=-0.0189, interpMode=RCIM_Cubic, leaveTangent=-0.0189, tangentMode=RCTM_User, time=0, value=1}
         {arriveTangent=-3.722, interpMode=RCIM_Cubic, leaveTangent=-3.722, tangentMode=RCTM_User, time=0.8, value=0}
        ]}}})
    Index 0 Param 4=&FloatFromCurve(FloatCurve=DataInterface() {propertyValues {Curve {keys [<at 0 0 RCIM_Cubic>,<at 1 1 RCIM_Cubic>]}}}, Scale Curve=0.8)
    Module.Index0Param0Write=true
    Module.Index0Param1Write=true
    Module.Index0Param2Write=true
    Module.Index0Param3Write=true
   )
  ]
  rendererClasses [&NiagaraSpriteRendererProperties]
  renderers [{rendererClass=&NiagaraSpriteRendererProperties}]
 }
 stone {
  emitterUpdateScript [
   &EmitterState(Inactive Response=::NewEnumerator1, Life Cycle Mode=::NewEnumerator1, Loop Behavior=::NewEnumerator1, Loop Duration=1, UseLoopDelay=false)
   &SpawnBurst_Instantaneous(Spawn Count=10, Use Spawn Probability=false, UseLoopCountLimit=false)
  ]
  particleSpawnScript [
   &InitializeParticle(
    Lifetime Max=1
    Lifetime Mode=::NewEnumerator1
    Sprite Rotation Mode=::NewEnumerator2
    Sprite Size Mode=::NewEnumerator1
    Uniform Sprite Size Max=7
    Uniform Sprite Size Min=3
    UsePositionOffset=false
    Write Material Random=true
   )
   &ShapeLocation(Hemisphere Distribution=<1 0.5>, MaskSpawnGroup=false, Override Seed=false, Sphere Radius=10, UseOwnerScale=true)
   &AddVelocity(
    Module.Use Falloff Distance=false
    Override Seed=false
    Velocity Mode=::NewEnumerator1
    Velocity Speed=&RandomRangeFloat(Minimum=100, Maximum=200)
   )
   &AddVelocity(Velocity=<0 0 100>) {module=AddVelocity001}
  ]
  particleUpdateScript [
   &ParticleState()
   &SpriteRotationRate(Rotation Rate=&Multiply_Float(A=&RandomRangeFloat(Minimum=-1, Maximum=1), B=800))
   &Drag(Use Linear Drag=true, Use Rotational Drag=true)
   &GravityForce()
   &SolveForcesAndVelocity(Clamp Velocity=false, Limit Acceleration=false)
   &ScaleColor(
    Linear Color Curve=DataInterface() {propertyValues {
      AlphaCurve {keys [<at 0 1>]}
      BlueCurve {keys [<at 0 0.1794>,<at 0.5 0.2448>]}
      GreenCurve {keys [<at 0 0.266>,<at 0.5 0.2448>]}
      RedCurve {keys [<at 0 0.3281>,<at 0.5 0.2448>]}
     }}
    Scale Mode=::NewEnumerator2
   )
  ]
  rendererClasses [&NiagaraSpriteRendererProperties]
  renderers [{rendererClass=&NiagaraSpriteRendererProperties}]
 }
}
summary {
 systemName=NS_Emote_SandalSite_ImpactSmall
 emitters [
  {emitterName=Glow, rendererClasses [&NiagaraSpriteRendererProperties, &NiagaraLightRendererProperties]}
  {emitterName=Streak, rendererClasses [&NiagaraSpriteRendererProperties]}
  {emitterName=Soil, rendererClasses [&NiagaraSpriteRendererProperties]}
  {emitterName=stone, rendererClasses [&NiagaraSpriteRendererProperties]}
  {emitterName=Smoke002, rendererClasses [&NiagaraSpriteRendererProperties]}
 ]
 userVariables [
  User.bIsFrontEnd {type=NiagaraBool}
  User.bIsFrontEndPreview {type=NiagaraBool}
 ]
}
