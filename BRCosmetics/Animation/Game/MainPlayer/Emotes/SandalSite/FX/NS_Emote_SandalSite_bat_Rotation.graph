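# Niagara effect for the SandalSite emote, drawing a spinning mesh effect that, by its name, represents bat rotation.
# The Rotation emitter bursts three particles, partitioned for distinct brown tints and staggered orientations, then spins them about negative Y.
# Two mesh renderers accompany separate Size002 and Color002 attributes, with color and alpha adapting to time of day and User.bIsFrontEndPreview.
# The Blur_LW emitter lists no renderers.
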NiagaraSystem /BRCosmetics/Animation/Game/MainPlayer/Emotes/SandalSite/FX/NS_Emote_SandalSite_bat_Rotation
dependencies {
 usedDataInterfaces [&NiagaraDataInterfaceArrayColor, &NiagaraDataInterfaceArrayFloat3]
 usedDynamicInputs []
 usedModules [
  &self:SystemSpawnScript.SystemScriptSource.SystemScriptGraph.NiagaraNodeAssignment_0.SetVariables_#1
  &SystemState
 ]
}
emitters {
 Blur_LW {rendererClasses [], renderers []}
 Rotation {
  emitterUpdateScript [
   &EmitterState(Inactive Response=::NewEnumerator1, Life Cycle Mode=::NewEnumerator1, Loop Behavior=::NewEnumerator1, Loop Duration=1, UseLoopDelay=false)
   &SpawnBurst_Instantaneous(Spawn Count=3, Use Spawn Probability=false, UseLoopCountLimit=false)
  ]
  particleSpawnScript [
   &PartitionParticles(Number Of Partitions=3)
   &InitializeParticle(
    Color=&SelectLinearColorFromArray(
     Array Sampling Mode=::NewEnumerator1
     Direct Array Index=$Particles.PartitionParticles.Partition
     Color Selection Array=DataInterface() {propertyValues {ColorData [<rgba 0.15 0.0894 0.045 1>, <rgba 0.3 0.1789 0.09 0.8>, <rgba 0.4 0.25 0.14 0.8>]}}
     Random Seed=?
    )
    Mesh Scale=&Multiply_Vector(A=&VectorFromFloat(Value=0.33), B=<1 2 1>)
    Mesh Scale Mode=::NewEnumerator4
    Position Offset=<0 10 0>
    UsePositionOffset=true
    Write Material Random=true
   )
   &InitialMeshOrientation(Mesh Orientation Mode=::NewEnumerator2, Rotation=<0 0 0.25>, Use Rotation Vector=true) {module=InitialMeshOrientation001}
   &InitialMeshOrientation(
    Mesh Orientation Mode=::NewEnumerator2
    Rotation=&SelectVectorFromArray(
     Direct Array Index=$Particles.PartitionParticles.Partition
     Vector Selection Array=DataInterface() {propertyValues {FloatData [<0 0 0>, <0.35 0 0>, <0.65 0 0>]}}
    )
    Use Rotation Vector=true
   )
  ]
  particleUpdateScript [
   &ParticleState()
   &UpdateMeshOrientation(Rotation Rate=3, Rotation Vector=<0 -1 0>)
   &ScaleColor(
    Scale Alpha=&MakeCustomFloatFromBool(Bool=$User.bIsFrontEndPreview, True Float=0.5, False Float=3)
    Scale RGB=&MakeVectorFromLinearColorRGB(Color=&FNTimeOfDayColor(Night Color=<rgba 1 1 1 1>, Morning Color=<rgba 1 1 1 1>, Day Color=<rgba 2 2 2 1>, Evening Color=<rgba 1.5 1.2825 1.1641 1>))
    ScaleA=true
    ScaleRGB=true
   )
   /SetVariables(Particles.Size002:Vector3f=&Multiply_Vector(A=&VectorFromFloat(Value=0.33), B=<1 1.6 1>)) {isSetParameters=true}
   /SetVariables(Particles.Color002=&MakeCustomLinearColorFromBool(
     Bool=$User.bIsFrontEndPreview
     True Color=<rgba 0.25 0.1563 0.0875 5>
     False Color=&FNTimeOfDayColor(
      Night Color=<rgba 0.15 0.0981 0.06 10>
      Morning Color=<rgba 0.3 0.1702 0.075 10>
      Day Color=<rgba 0.3 0.1702 0.075 12>
      Evening Color=<rgba 0.25 0.1563 0.0875 10>
     )
    )) {isSetParameters=true}
  ]
  rendererClasses [&NiagaraMeshRendererProperties]
  renderers [
   {rendererClass=&NiagaraMeshRendererProperties, rendererIndex=0}
   {rendererClass=&NiagaraMeshRendererProperties, rendererIndex=1}
  ]
 }
}
summary {
 systemName=NS_Emote_SandalSite_bat_Rotation
 emitters [{emitterName=Rotation, rendererClasses [&NiagaraMeshRendererProperties]}, {emitterName=Blur_LW, rendererClasses []}]
 userVariables [User.bIsFrontEndPreview {type=NiagaraBool}]
}
