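# Niagara ribbon trail effect for the bat in the SandalSite emote.
# A single short-lived emitter spawns ribbon particles at a steady rate with a fixed facing vector; spawn-per-unit and added velocity are disabled.
# Color curves darken a warm brown and width tapers over the ribbon.
# Alpha is indexed by emitter age rather than particle age, fading the whole trail early, with a time-of-day scalar adjusting its strength.
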NiagaraSystem /BRCosmetics/Animation/Game/MainPlayer/Emotes/SandalSite/FX/NS_Emote_SandalSite_bat_Trail
dependencies {
 usedDataInterfaces [&NiagaraDataInterfaceColorCurve, &NiagaraDataInterfaceCurve]
 usedDynamicInputs []
 usedModules [&SystemState]
}
emitters {Trail {
  emitterUpdateScript [
   &EmitterState(Life Cycle Mode=::NewEnumerator1, Loop Behavior=::NewEnumerator1, Loop Duration=1, UseLoopDelay=false)
   &SpawnPerUnit(
    Max Movement Threshold=2000
    Module.Use Max Movement Threshold=true
    Module.Use Movement Tolerance=false
    Spawn Spacing=40
    Use Spawn Probability=false
   ) {enabled=false}
   &SpawnRate(SpawnRate=30, Use Spawn Probability=false)
  ]
  particleSpawnScript [
   &InitializeParticle(Lifetime=0.3, Ribbon Width=-60, Ribbon Width Mode=::NewEnumerator1, UsePositionOffset=false, Write Material Random=true)
   &self:RibbonFace(Vector:Vector3f=<0 1 0>) {module=RibbonFace}
   &AddVelocity() {enabled=false}
  ]
  particleUpdateScript [
   &ParticleState()
   &ScaleColor(
    Scale Alpha=&FloatFromCurve(
     FloatCurve=DataInterface() {propertyValues {Curve {keys [
         {arriveTangent=-4.5685, interpMode=RCIM_Cubic, leaveTangent=-4.5685, tangentMode=RCTM_User, time=0.1, value=1}
         {arriveTangent=-1.1502, interpMode=RCIM_Cubic, leaveTangent=-1.1502, tangentMode=RCTM_User, time=0.4, value=0}
        ]}}}
     CurveIndex=$Emitter.Age
     Scale Curve=&FNTimeOfDayScalar(Night Scale=1, Morning Scale=2, Day Scale=2, Evening Scale=2)
    )
    Scale RGB=&MakeVectorFromLinearColorRGB(Color=&ColorFromCurve(ColorCurve=DataInterface() {propertyValues {
        AlphaCurve {keys []}
        BlueCurve {keys [<at 0.1694 0.0984 RCIM_Cubic>,<at 0.5721 0.0163>]}
        GreenCurve {keys [<at 0.1694 0.2739 RCIM_Cubic>,<at 0.5721 0.0601>]}
        RedCurve {keys [<at 0.1694 0.4635 RCIM_Cubic>,<at 0.5721 0.099>]}
       }}))
    ScaleA=true
    ScaleRGB=true
   )
   &SolveForcesAndVelocity(Clamp Velocity=false, Limit Acceleration=false)
   &ScaleRibbonWidth(Ribbon Width Scale=&FloatFromCurve(FloatCurve=DataInterface() {propertyValues {Curve {keys [{arriveTangent=-2.3008, time=0.3, value=1}, <at 1 0.4 RCIM_Cubic>]}}}))
  ]
  rendererClasses [&NiagaraRibbonRendererProperties]
  renderers [{rendererClass=&NiagaraRibbonRendererProperties}]
 }}
summary {
 systemName=NS_Emote_SandalSite_bat_Trail
 emitters [{emitterName=Trail, rendererClasses [&NiagaraRibbonRendererProperties]}]
 userVariables [
  User.bIsFrontEnd {type=NiagaraBool}
  User.bIsFrontEndPreview {type=NiagaraBool}
 ]
}
